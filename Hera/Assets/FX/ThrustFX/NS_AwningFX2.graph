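# Niagara thrust effect (NS_AwningFX2) that draws downward dust, faint heat and wind streams on either side of a center point along X.
# All three emitters place particles using shared system spread and offset values; the system opacity multiplier affects only Dust and Side_Wind.
# Dust and Heat are short-lived sprites with distance-driven material parameters, with Dust alpha also scaled by time of day.
# Side_Wind bursts two persistent meshes with waveform-pulsed Z scale and an age-driven material parameter.
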NiagaraSystem /Hera/Assets/FX/ThrustFX/NS_AwningFX2
dependencies {
 usedDataInterfaces [
  &NiagaraDataInterfaceCurve
  &NiagaraDataInterfaceColorCurve
  &NiagaraDataInterfaceArrayFloat3
  &NiagaraDataInterfaceSpriteRendererInfo
 ]
 usedDynamicInputs []
 usedModules [
  &self:SystemSpawnScript.SystemScriptSource.SystemScriptGraph.NiagaraNodeAssignment_0.SetVariables_#1
  &self:SystemSpawnScript.SystemScriptSource.SystemScriptGraph.NiagaraNodeAssignment_1.SetVariables_#2
  &self:SystemSpawnScript.SystemScriptSource.SystemScriptGraph.NiagaraNodeAssignment_2.SetVariables_#3
  &SystemState
 ]
}
emitters {
 Dust {
  emitterUpdateScript [
   &EmitterState()
   &SpawnRate(SpawnRate=15, Use Spawn Probability=false)
  ]
  particleSpawnScript [
   &InitializeParticle(
    Color Mode=::NewEnumerator0
    Lifetime Max=0.75
    Lifetime Mode=::NewEnumerator1
    Mass Mode=::NewEnumerator3
    Sprite Rotation Mode=::NewEnumerator2
    Sprite Size Mode=::NewEnumerator3
    Sprite UV Mode=::NewEnumerator3
    Uniform Sprite Size=65
    UsePositionOffset=false
    Write Material Random=true
   )
   &PartitionParticles(Number Of Partitions=4, Write Partition to Ribbon ID=::NewEnumerator0)
   &SystemLocation(
    MaskSpawnGroup=false
    Offset=&Add_Vector(
     A=&Multiply_Vector(
      A=&SelectVectorFromArray(
       Direct Array Index=$Particles.PartitionParticles.Partition
       Direct Array Mode=::NewEnumerator1
       Vector Selection Array=DataInterface() {propertyValues {FloatData [<1 1 0>, <-1 1 1>, <-1 -1 0>, <1 -1 0>]}}
      )
      B=&MakeVector(X=$System.SysSpread, Y=0, Z=1)
     )
     B=$System.OverallOffset
    )
   )
   &InitialMeshOrientation(
    Mesh Orientation Mode=::NewEnumerator2
    Rotation=&MaskVectorBySpawnGroup(
     Masked Vector=&MakeVector(X=0, Y=0, Z=&Add_Float(A=0.055, B=$Particles.PartitionParticles.Normalized Index In Partition))
     Passthrough Vector=&MakeVector(X=0, Y=0, Z=&Add_Float(A=-0.055, B=$Particles.PartitionParticles.Normalized Index In Partition))
    )
    Rotation Coordinate Space=::NewEnumerator0
    Use Rotation Vector=true
   )
   &AddVelocity(Velocity=<0 0 -1>, Velocity Speed Scale=120)
   &CameraOffset(Camera Offset Amount=25)
   /SetVariables() {isSetParameters=true}
  ]
  particleUpdateScript [
   &ParticleState()
   &ScaleSpriteSize(
    Scale Factor=&Lerp_Vector2(
     A=<1.25 1.25>
     B=<1 1>
     Alpha=&FloatFromCurve(FloatCurve=DataInterface() {propertyValues {Curve {
         defaultValue=1
         keys [{interpMode=RCIM_Cubic, tangentMode=RCTM_User, time=0, value=0}, {interpMode=RCIM_Cubic, tangentMode=RCTM_User, time=1, value=1}]
        }}})
    )
    Scale Sprite Size Mode=::NewEnumerator2
   )
   &SpriteRotationRate(Rotation Rate=&Multiply_Float(A=90, B=&RandomRangeFloat(Minimum=-1, Maximum=1)))
   &AccelerationForce()
   &Drag(Drag=0.5, Use Linear Drag=true, Use Rotational Drag=true)
   &SolveForcesAndVelocity(Clamp Velocity=false, Limit Acceleration=false)
   &DynamicMaterialParameters(
    Index 3 Param 4=&FloatFromCurve(
     FloatCurve=DataInterface() {propertyValues {Curve {keys [
         <at 0 1 RCIM_Cubic>
         <at 0.1388 1.15>
         {arriveTangent=-1.2839, interpMode=RCIM_Cubic, leaveTangent=-1.2839, tangentMode=RCTM_User, time=1, value=1}
        ]}}}
     CurveIndex=&FN_LOD_Distance_Remap(Start Distance=0, End Distance=15000, Minimum Output=0, Maximum Output=1, Clamp Output=true)
    )
    Module.Index3Float0Write=false
    Module.Index3Float1Write=false
    Module.Index3Float2Write=false
    Module.Index3Float3Write=true
    Write Parameter Index 0=::NewEnumerator0
    Write Parameter Index 3=::NewEnumerator1
   )
   &Color(
    Color=&ColorFromCurve(ColorCurve=DataInterface() {propertyValues {
       AlphaCurve {
        defaultValue=1
        keys [
         {arriveTangent=9.0725, interpMode=RCIM_Cubic, leaveTangent=9.0725, tangentMode=RCTM_User, time=0, value=0}
         {
          arriveTangent=-1.9008
          arriveTangentWeight=0.0469
          interpMode=RCIM_Cubic
          leaveTangent=-1.9008
          tangentMode=RCTM_User
          time=0.303
          value=0.9269
         }
         {
          arriveTangent=-0.7375
          arriveTangentWeight=0.2562
          interpMode=RCIM_Cubic
          leaveTangent=-0.7375
          leaveTangentWeight=0.2562
          tangentMode=RCTM_User
          tangentWeightMode=RCTWM_WeightedBoth
          time=1
          value=0
         }
        ]
       }
       BlueCurve {keys [<at 0.0613 1>,<at 0.4647 1 RCIM_Cubic>,<at 1 1 RCIM_Cubic>]}
       GreenCurve {keys [<at 0.0613 0.9296>, {arriveTangent=0.075, interpMode=RCIM_Cubic, leaveTangent=0.075, time=0.4647, value=1}, <at 1 1 RCIM_Cubic>]}
       RedCurve {keys [<at 0.0613 0.9>, {arriveTangent=0.1065, interpMode=RCIM_Cubic, leaveTangent=0.1065, time=0.4647, value=1}, <at 1 1 RCIM_Cubic>]}
      }})
    Scale Alpha=&Multiply_Float(
     A=&Multiply_Float(A=0.8, B=$System.Opac Mul)
     B=&FNTimeOfDayScalar(Night Scale=0.12, Morning Scale=0.06, Day Scale=0.06, Evening Scale=0.06)
    )
   )
   &SubUVAnimation(
    Random Start Frame=true
    "Sprite Renderer "=DataInterface() {propertyValues {SpriteRenderer=&self:Dust_0.NiagaraSpriteRendererProperties_1}}
    SubUV Lookup Index Scale=0.5
    UseEndFrame=false
    UseStartFrame=false
   )
  ]
  rendererClasses [&NiagaraSpriteRendererProperties]
  renderers [{rendererClass=&NiagaraSpriteRendererProperties}]
 }
 Heat {
  emitterUpdateScript [
   &EmitterState()
   &SpawnRate(SpawnRate=10, Use Spawn Probability=false)
  ]
  particleSpawnScript [
   &InitializeParticle(
    Color Mode=::NewEnumerator0
    Lifetime Max=0.75
    Lifetime Mode=::NewEnumerator1
    Mass Mode=::NewEnumerator3
    Sprite Rotation Mode=::NewEnumerator2
    Sprite Size Mode=::NewEnumerator3
    Uniform Sprite Size=50
    UsePositionOffset=false
    Write Material Random=true
   )
   &PartitionParticles(Write Partition to Ribbon ID=::NewEnumerator0)
   &SystemLocation(
    MaskSpawnGroup=false
    Offset=&Add_Vector(
     A=&Add_Vector(
      A=&Multiply_VectorByFloat(
       Vector=&Lerp_Vector(A=<-1 0 0>, B=<1 0 0>, Alpha=&MakeFloatFromInt(Int=$Particles.PartitionParticles.Partition))
       Float=$System.SysSpread
      )
      B=$System.OverallOffset
     )
     B=<0 0 15>
    )
   )
   &InitialMeshOrientation(
    Mesh Orientation Mode=::NewEnumerator2
    Rotation=&MaskVectorBySpawnGroup(
     Masked Vector=&MakeVector(X=0, Y=0, Z=&Add_Float(A=0.055, B=$Particles.PartitionParticles.Normalized Index In Partition))
     Passthrough Vector=&MakeVector(X=0, Y=0, Z=&Add_Float(A=-0.055, B=$Particles.PartitionParticles.Normalized Index In Partition))
    )
    Rotation Coordinate Space=::NewEnumerator0
    Use Rotation Vector=true
   )
   &AddVelocity(Velocity=<0 0 -1>, Velocity Speed Scale=150)
   &CameraOffset()
  ]
  particleUpdateScript [
   &ParticleState()
   &SpriteRotationRate(Rotation Rate=&Multiply_Float(A=30, B=&RandomRangeFloat(Minimum=-1, Maximum=1)))
   &AccelerationForce()
   &Drag(Drag=0.5, Use Linear Drag=true, Use Rotational Drag=true)
   &SolveForcesAndVelocity(Clamp Velocity=false, Limit Acceleration=false)
   &DynamicMaterialParameters(
    Index 3 Param 4=&FloatFromCurve(
     FloatCurve=DataInterface() {propertyValues {Curve {keys [
         <at 0 1 RCIM_Cubic>
         <at 0.1388 1.15>
         {arriveTangent=-1.2839, interpMode=RCIM_Cubic, leaveTangent=-1.2839, tangentMode=RCTM_User, time=1, value=1}
        ]}}}
     CurveIndex=&FN_LOD_Distance_Remap(Start Distance=0, End Distance=15000, Minimum Output=0, Maximum Output=1, Clamp Output=true)
    )
    Module.Index3Float0Write=false
    Module.Index3Float1Write=false
    Module.Index3Float2Write=false
    Module.Index3Float3Write=true
    Write Parameter Index 0=::NewEnumerator0
    Write Parameter Index 3=::NewEnumerator1
   )
   &Color(
    Color=&ColorFromCurve(ColorCurve=DataInterface() {propertyValues {
       AlphaCurve {
        defaultValue=1
        keys [
         {arriveTangent=9.0725, interpMode=RCIM_Cubic, leaveTangent=9.0725, tangentMode=RCTM_User, time=0, value=0}
         {
          arriveTangent=-2.5249
          arriveTangentWeight=0.0469
          interpMode=RCIM_Cubic
          leaveTangent=-2.5249
          tangentMode=RCTM_User
          time=0.2414
          value=0.9269
         }
         {
          arriveTangent=-0.0405
          arriveTangentWeight=0.4326
          interpMode=RCIM_Cubic
          leaveTangent=-0.0405
          leaveTangentWeight=0.4326
          tangentMode=RCTM_User
          tangentWeightMode=RCTWM_WeightedBoth
          time=1
          value=0
         }
        ]
       }
       BlueCurve {keys [<at 0.2333 1>,<at 0.6718 1 RCIM_Cubic>,<at 1 1 RCIM_Cubic>]}
       GreenCurve {keys [<at 0.2333 0.9296>, {arriveTangent=0.0918, interpMode=RCIM_Cubic, leaveTangent=0.0918, time=0.6718, value=1}, <at 1 1 RCIM_Cubic>]}
       RedCurve {keys [<at 0.2333 0.9>, {arriveTangent=0.1304, interpMode=RCIM_Cubic, leaveTangent=0.1304, time=0.6718, value=1}, <at 1 1 RCIM_Cubic>]}
      }})
    Scale Alpha=0.035
   )
  ]
  rendererClasses [&NiagaraSpriteRendererProperties]
  renderers [
   {rendererClass=&NiagaraSpriteRendererProperties}
   {rendererClass=&NiagaraSpriteRendererProperties, rendererIndex=1}
  ]
 }
 Side_Wind {
  emitterUpdateScript [
   &EmitterState()
   &SpawnBurst_Instantaneous(Spawn Count=2, Use Spawn Probability=false, UseLoopCountLimit=true)
  ]
  particleSpawnScript [
   &InitializeParticle(
    Color=&MultiplyLinearColor(
     Linear Color=<rgba 0.9 0.9296 1 0.15>
     Scale Factor=&MakeVector4FromVectorAndFloat("Vector (RGB)"=<1 1 1>, "Float (Alpha)"=$System.Opac Mul)
    )
    Lifetime Max=1.8
    Lifetime Min=2
    Lifetime Mode=::NewEnumerator1
    Mass Mode=::NewEnumerator3
    Mesh Scale=<0.45 0.45 1.35>
    Mesh Scale Mode=::NewEnumerator4
    Sprite Rotation Mode=::NewEnumerator2
    Sprite Size Mode=::NewEnumerator3
    Uniform Sprite Size=750
    UsePositionOffset=false
    Write Material Random=true
   )
   &PartitionParticles(Write Partition to Ribbon ID=::NewEnumerator0)
   &SystemLocation(
    MaskSpawnGroup=false
    Offset=&Add_Vector(
     A=&Multiply_VectorByFloat(
      Vector=&Lerp_Vector(A=<-1 0 0>, B=<1 0 0>, Alpha=&MakeFloatFromInt(Int=$Particles.PartitionParticles.Partition))
      Float=$System.SysSpread
     )
     B=$System.OverallOffset
    )
   )
   &InitialMeshOrientation(Mesh Orientation Mode=::NewEnumerator2, Rotation=<-0.25 0 0>, Rotation Coordinate Space=::NewEnumerator0, Use Rotation Vector=true)
   /SetVariables(Particles.Linear Color=&MultiplyLinearColor(
     Linear Color=<rgba 0.9 0.9296 1 0.35>
     Scale Factor=&MakeVector4FromVectorAndFloat("Vector (RGB)"=<1 1 1>, "Float (Alpha)"=$System.Opac Mul)
    )) {isSetParameters=true}
  ]
  particleUpdateScript [
   &ParticleState(Kill Particles When Lifetime Has Elapsed=false, Let Infinitely Lived Particles Die When Emitter Deactivates=true)
   &ScaleMeshSize(Scale Factor=&MakeVector(
     X=1
     Y=1
     Z=&Waveform(
      "Waveform [1]"=::NewEnumerator12
      "[1] Input Position"=&RandomRangeFloat(Minimum=0, Maximum=10000)
      "[1] Phase Scale"=4
      "[1] Frequency"=1
      "[1] Amplitude Min/Max"=<0.7 1>
     )
    ))
   &DynamicMaterialParameters(
    Index 1 Param 1=&Add_Float(
     A=&Multiply_Float(A=$Emitter.Age, B=0.75)
     B=&RandomRangeFloat(Minimum=0, Maximum=100)
    )
    Index 1 Param 2=1
    Module.Index1Float0Write=true
    Module.Index1Float1Write=false
    Module.Index1Float2Write=false
    Module.Index1Float3Write=false
    Write Parameter Index 0=::NewEnumerator0
    Write Parameter Index 1=::NewEnumerator1
   ) {module=DynamicMaterialParameters001}
  ]
  rendererClasses [&NiagaraMeshRendererProperties]
  renderers [
   {rendererClass=&NiagaraMeshRendererProperties, rendererIndex=0}
   {rendererClass=&NiagaraMeshRendererProperties, rendererIndex=1}
  ]
 }
}
summary {
 systemName=NS_AwningFX2
 emitters [
  {emitterName=Heat, rendererClasses [&NiagaraSpriteRendererProperties]}
  {emitterName=Side_Wind, rendererClasses [&NiagaraMeshRendererProperties]}
  {emitterName=Dust, rendererClasses [&NiagaraSpriteRendererProperties]}
 ]
 userVariables []
}
